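# Material function for the Octane Lenderman car skin, outputting material attributes for a glossy, fully metallic paint.
# A Fresnel term blends BaseColor to AccentColor and, with reflection-vector shading, shapes base color plus an added rim.
# Emissive fakes reflections from a falloff texture array, with row chosen by F1Type, and an environment texture sampled by reflection vector.
# F1ControlA and F1ControlB vectors tune rim tint, emissive tint and reflection strengths.
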MaterialFunction /VehicleCosmetics/Bodies/Octane/Skins/OctaneLenderman/MF_Skin_Lenderman
graph {nodes [
  DotProduct(
   A=Transform(None=CameraVectorWS()) {transformSourceType=TRANSFORMSOURCE_World, transformType=TRANSFORM_Tangent}
   B=Constant3Vector() {constant=<rgba 0.1 -0.1 0.6 0>}
  )
  n1 = VectorParameter() {group=Color, parameterName=AccentColor, sortPriority=2, defaultValue=<rgba 1 1 1 1>}
  n2 = Fresnel() {baseReflectFraction=0, exponent=3}
  n3 = ConstantBiasScale(None=@n2) {bias=-0.25, scale=2}
  n4 = LinearInterpolate(
   A=VectorParameter() {group=Color, parameterName=BaseColor, sortPriority=1, defaultValue=<rgba 0 0.2 1 1>}.RGB
   B=@n1.RGB
   Alpha=Saturate(None=@n3)
  )
  n5 = ReflectionVectorWS()
  n6 = Constant3Vector() {constant=<rgba 0.3 0.2 0 0>}
  n7 = Add(A=Multiply(A=ScalarParameter() {defaultValue=7, parameterName=F1Type}) {constB=0.0312}) {constB=0.0039}
  n8 = TextureSample(UVs=AppendVector(A=DotProduct(A=@n5, B=@n6), B=Add(A=@n7) {constB=0.0156})) {texture=/VehicleCosmetics/Textures/UniversalTextures/T_LightFalloffArray}
  n9 = ConstantBiasScale(None=@n8.R) {bias=-0.22, scale=8}
  n10 = Max(A=@n3, B=Clamp(None=Add(A=@n9) {constB=0.5}))
  n11 = ConstantBiasScale(None=DotProduct(A=ReflectionVectorWS(), B=Constant3Vector() {constant=<rgba 0.5 0.5 1.2 0>})) {bias=-0.2, scale=1}
  n12 = Saturate(None=@n11)
  n13 = Multiply(A=Multiply(A=@n4, B=Saturate(None=@n10)), B=Saturate(None=ConstantBiasScale(None=@n12) {bias=0.3, scale=1}))
  n14 = VectorParameter() {desc=Material Parameters A, parameterName=F1ControlA, defaultValue=<rgba 0.96 40 1 0>}
  n15 = Saturate(None=Multiply(A=LinearInterpolate(B=@n1.RGB, Alpha=@n14.G), B=ConstantBiasScale(None=@n2) {bias=-0.8, scale=5}))
  n16 = LinearInterpolate(B=@n4, Alpha=@n14.R) {constA=0.5, constB=0.5, desc=DiffsueSpecMultiplyAmount}
  n17 = ConstantBiasScale(None=@n8.R) {bias=0, scale=1}
  n18 = Power(Base=@n11) {constExponent=1}
  n19 = Clamp(None=Max(A=ConstantBiasScale(None=@n18) {bias=-0.6, scale=0.2}, B=ConstantBiasScale(None=@n18) {bias=-0.99, scale=1})) {minDefault=-0.05}
  n20 = VectorParameter() {parameterName=F1ControlB, defaultValue=<rgba 300 20 0 1>}
  n21 = ReflectionVectorWS()
  n22 = TextureSample(UVs=ComponentMask(None=ConstantBiasScale(None=@n21)) {g=true, r=true}) {texture=/VehicleCosmetics/Textures/UniversalTextures/T_ENVPack}
  n23 = LinearInterpolate(A=@n22.R, B=@n22.R) {constAlpha=0.4}
  n24 = Add(
   A=Multiply(A=Multiply(A=@n17, B=@n19), B=@n20.R)
   B=Multiply(A=ConstantBiasScale(None=@n8.R) {bias=0.75, scale=1}, B=Multiply(A=@n23, B=@n20.G))
  )
  n25 = SetMaterialAttributes!(
   MaterialAttributes
   BaseColor=Clamp(None=Add(A=@n13, B=@n15))
   Metallic=Constant() {r=1}
   Roughness=Constant() {r=0.1}
   EmissiveColor=Multiply(A=Multiply(A=@n16, B=@n24), B=Constant() {desc=Emissive, r=0.75})
   Normal
   AmbientOcclusion
  ) {attributeSetTypes [
    #1
    #2
    #3
    #4
    #5
    #6
   ]}
  FunctionOutput(None=@n25)
 ]}
